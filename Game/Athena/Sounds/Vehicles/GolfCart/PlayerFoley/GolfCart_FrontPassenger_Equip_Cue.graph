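# Sound cue for the foley heard when a front passenger equips in a golf cart.
# A single wave player feeds the passenger equip wave straight to the output, with no randomization or modulation.
# It plays at reduced volume in the vehicles sound class, spatialized by the close golf cart movement attenuation.
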SoundCue /Game/Athena/Sounds/Vehicles/GolfCart/PlayerFoley/GolfCart_FrontPassenger_Equip_Cue
root=WavePlayer_1
nodes [{id=WavePlayer_1, properties {soundWaveAssetPtr=&/GolfCart_Equip_Passenger}}]
properties {
 attenuationSettings=/Game/Sounds/Attenuation/Fort_GolfCart_Movement_Close_Att
 maxDistance=5300
 soundClassObject=&Fort_Vehicles
 volumeMultiplier=0.3
}
