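# Sound cue for player foley when switching from the driver seat to a passenger seat in a golf cart.
# A single wave player feeds the root directly, playing one seat-switching sound wave with no randomization or modulation.
# It routes through the Fortnite vehicles sound class and uses a close-range golf cart movement attenuation.
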SoundCue /Game/Athena/Sounds/Vehicles/GolfCart/PlayerFoley/SeatSwitchingDriverToPassengerBR_GolfCart_M_Cue
root=WavePlayer_0
nodes [{
  id=WavePlayer_0
  properties {soundWaveAssetPtr=&/SeatSwitchingDriverToPassengerBR_GolfCart_M}
 }]
properties {
 attenuationSettings=/Game/Sounds/Attenuation/Fort_GolfCart_Movement_Close_Att
 maxDistance=5300
 soundClassObject=&Fort_Vehicles
}
